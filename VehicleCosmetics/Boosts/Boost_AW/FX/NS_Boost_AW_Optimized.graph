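# Optimized Niagara effect for the AW vehicle boost cosmetic: an orange exhaust cone, smoke and fluttering ticket sprites.
# Cone bursts two looping sprites, smoke plays SubUV animation with camera fade, and Tickets feeds two spawn groups into visibility-tagged renderers with curl noise and gravity.
# User.GameModeID switches smoke and ticket motion.
# Placement uses SingleBoostPos and DTCarOffset user vectors, with skeletal socket sampling and painted-variant overrides disabled.
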NiagaraSystem /VehicleCosmetics/Boosts/Boost_AW/FX/NS_Boost_AW_Optimized
dependencies {
 usedDataInterfaces [
  &NiagaraDataInterfaceArrayColor
  &NiagaraDataInterfaceArrayFloat
  &NiagaraDataInterfaceCurve
  &NiagaraDataInterfaceVectorCurve
  &NiagaraDataInterfaceCamera
  &NiagaraDataInterfaceSpriteRendererInfo
  &NiagaraDataInterfaceArrayFloat2
  &NiagaraDataInterfaceVector2DCurve
  &NiagaraDataInterfaceArrayFloat3
  &NiagaraDataInterfaceSkeletalMesh
 ]
 usedDynamicInputs []
 usedModules [
  &self:SystemSpawnScript.SystemScriptSource.SystemScriptGraph.NiagaraNodeAssignment_0.SetVariables_#1
  &SystemState
 ]
}
emitters {
 Cone {
  emitterUpdateScript [
   &EmitterState(
    Inactive Response=::NewEnumerator1
    Life Cycle Mode=::NewEnumerator1
    Loop Behavior=::NewEnumerator1
    Loop Duration Mode=::NewEnumerator1
    UseLoopDelay=false
   )
   &SpawnBurst_Instantaneous(Spawn Count=2, Use Spawn Probability=false, UseLoopCountLimit=false)
   &VC_FilteredSocketCount()
   &SpawnRate(SpawnRate=50, Use Spawn Probability=false) {enabled=false}
  ]
  particleSpawnScript [
   &VC_PaintedVariantsOverride() {enabled=false}
   &InitializeParticle(
    Color=<rgba 1.2 0.22 0 0.9>
    Lifetime=0
    Position Offset=&AddVectorToPosition(
     Position=&AddVectorToPosition(Position=&AddVectorToPosition(Position=<-40 0 0>, Vector=<-20 0 0>), Vector=$User.SingleBoostPos)
     Vector=$User.DTCarOffset
    )
    Sprite Rotation Angle=0.25
    Sprite Rotation Mode=::NewEnumerator1
    Sprite Size=&SelectVector2DFromArray(
     Direct Array Index=&ReturnExecIndex()
     Vector 2D Selection Array=DataInterface() {propertyValues {FloatData [<97.6 71.5>, <130.1 71.5>]}}
    )
    Sprite Size Mode=::NewEnumerator4
    UsePositionOffset=true
    Write Material Random=true
   )
   &PartitionParticles(Number Of Partitions=$Emitter.PortCount) {enabled=false}
   &SkeletalMeshLocation(
    Bone / Socket Index=$Particles.PartitionParticles.Partition
    Bone / Socket Sampling Mode=::NewEnumerator5
    Interpolate Mesh Sampling=false
    MaskSpawnGroup=false
    Mesh Sampling Type=::NewEnumerator6
    Orientation Sampling=::NewEnumerator1
    Sampled Position Offset=&AddVectorToPosition(Position=<0 0 0>, Vector=<-20 0 0>)
    Sampled Position Offset Coordinate Space:ENiagaraSkelMeshTransforms=::NewEnumerator2
    Skeletal Mesh=$System.SkeletalMesh
   ) {enabled=false}
   &AddVelocity(Velocity=<-1500 0 0>) {enabled=false}
   &CameraOffset(Camera Offset Amount=15)
   &SpriteFacingAndAlignment(Sprite Alignment=<-1 0 0>, WriteAlignment=true, WriteFacing=false)
  ]
  particleUpdateScript [
   &ParticleState(Kill Particles When Lifetime Has Elapsed=false, Loop Particles Lifetime=true)
   &SolveForcesAndVelocity(Clamp Velocity=false, Limit Acceleration=false)
   &ScaleColor(
    Scale Alpha=&FloatFromCurve(FloatCurve=DataInterface() {propertyValues {Curve {keys [<at 0 1>,<at 6 0>]}}})
    ScaleA=true
    ScaleRGB=true
   )
   &DynamicMaterialParameters(
    Index 0 Param 1=/VehicleCosmetics/FX/Niagara/Modules/MergeEmitters/VC_FloatBySpawnGroup(
     SpawnGroup or ExecIndex:NiagaraInt32=1
     CustomInt:NiagaraInt32=0
     MaximumSpawnGroups:NiagaraInt32=2
     Float SG 0:NiagaraFloat=&RandomRangeFloat(Minimum=1, Maximum=2)
     Float SG 1:NiagaraFloat=&RandomRangeFloat(Minimum=2, Maximum=3)
     Float SG 2:NiagaraFloat=0
     Float SG 3:NiagaraFloat=0
     Float SG 4:NiagaraFloat=0
     Float SG 5:NiagaraFloat=0
    )
    Index 0 Param 3=&VC_FloatBySpawnGroup(
     SpawnGroup or ExecIndex:NiagaraInt32=1
     CustomInt:NiagaraInt32=0
     MaximumSpawnGroups:NiagaraInt32=2
     Float SG 0:NiagaraFloat=0.5
     Float SG 1:NiagaraFloat=0.7
     Float SG 2:NiagaraFloat=0
     Float SG 3:NiagaraFloat=0
     Float SG 4:NiagaraFloat=0
     Float SG 5:NiagaraFloat=0
    )
    Module.Index0Param0Write=true
    Module.Index0Param1Write=true
    Module.Index0Param2Write=true
    Module.Index0Param3Write=true
   )
  ]
  rendererClasses [&NiagaraSpriteRendererProperties]
  renderers [{rendererClass=&NiagaraSpriteRendererProperties}]
 }
 NE_Smoke001 {
  emitterUpdateScript [
   &EmitterState()
   &VC_FilteredSocketCount()
   &SpawnRate(SpawnRate=80, Use Spawn Probability=false)
  ]
  particleSpawnScript [
   &VC_PaintedVariantsOverride() {enabled=false}
   &InitializeParticle(
    Color=&SelectLinearColorFromArray(
     Array Sampling Mode=::NewEnumerator0
     Direct Array Index=?
     Color Selection Array=DataInterface() {propertyValues {ColorData [<rgba 0.5 0.0833 0 1>, <rgba 1 0.2667 0 1>]}}
     Random Seed=0
    )
    Lifetime=&MergeHelper_FloatSwitch(
     0=&RandomRangeFloat(Minimum=0.8, Maximum=1)
     1=&RandomRangeFloat(Minimum=1.4, Maximum=1.6)
     2=&RandomRangeFloat(Minimum=1.4, Maximum=1.6)
     3=?
     Index=$User.GameModeID
     Num=::NewEnumerator1
    )
    Position Offset=&AddVectorToPosition(
     Position=&AddVectorToPosition(Position=&AddVectorToPosition(Position=<-40 0 0>, Vector=<-14 0 0>), Vector=$User.SingleBoostPos)
     Vector=$User.DTCarOffset
    )
    Position Offset Coordinate Space=::Local
    Sprite Rotation Mode=::NewEnumerator2
    Sprite Size Mode=::NewEnumerator1
    Uniform Sprite Size Max=100
    Uniform Sprite Size Min=80
    UsePositionOffset=true
    Write Material Random=true
   )
   &PartitionParticles(Number Of Partitions=$Emitter.PortCount) {enabled=false}
   &SkeletalMeshLocation(
    Bone / Socket Index=$Particles.PartitionParticles.Partition
    Bone / Socket Sampling Mode=::NewEnumerator5
    MaskSpawnGroup=false
    Mesh Sampling Type=::NewEnumerator6
    Sampled Position Offset=&AddVectorToPosition(Position=<0 0 0>, Vector=<-14 0 0>)
    Skeletal Mesh=$System.SkeletalMesh
   ) {enabled=false}
   &AddVelocity(Velocity=&MergeHelper_VectorSwitch(
     0=&RandomRangeVector(Minimum=<-1000 -100 -70>, Maximum=<-1000 100 100>)
     1=&RandomRangeVector(Minimum=<-500 -100 -70>, Maximum=<-500 100 100>)
     2=&RandomRangeVector(Minimum=<-500 -100 -70>, Maximum=<-500 100 100>)
     3=?
     Index=$User.GameModeID
     Num=::NewEnumerator1
    ))
  ]
  particleUpdateScript [
   &ParticleState()
   &SpriteRotationRate(Rotation Rate=&SelectFloatFromArray(
     Array Sampling Mode=::NewEnumerator0
     Direct Array Index=?
     Float Selection Array=DataInterface() {propertyValues {FloatData [-3,3]}}
     Random Seed=0
    ))
   &ScaleSpriteSize(Uniform Curve Sprite Scale=DataInterface() {propertyValues {Curve {keys [{tangentMode=RCTM_User, time=0, value=1.22}, {tangentMode=RCTM_User, time=0.5, value=2}]}}})
   &InheritSourceMovement(
    Applied Position Scale=&VectorFromCurve(VectorCurve=DataInterface() {propertyValues {
       XCurve {keys [<at 0.3 1 RCIM_Cubic>,<at 1 0.5 RCIM_Cubic>]}
       YCurve {keys [<at 0.3 1 RCIM_Cubic>,<at 1 0.5 RCIM_Cubic>]}
       ZCurve {keys [<at 0.3 1 RCIM_Cubic>,<at 1 0.5 RCIM_Cubic>]}
      }})
    ApplyForce=false
    ApplyPosition=true
    ApplyPositionFalloff=false
   )
   &ScaleColor(
    Scale Alpha=/Niagara/DynamicInputs/Camera/CameraFade(
     Camera Query:NiagaraDataInterfaceCamera=DataInterface()
     Fade Offset:NiagaraFloat=&MergeHelper_FloatSwitch(0=250, 1=100, 2=100, 3=?, Index=$User.GameModeID, Num=::NewEnumerator1)
     Fade Distance:NiagaraFloat=&MergeHelper_FloatSwitch(0=350, 1=150, 2=150, 3=?, Index=$User.GameModeID, Num=::NewEnumerator1)
     Use Owner Position:NiagaraBool=true
    )
    ScaleA=true
    ScaleRGB=false
   )
   &SolveForcesAndVelocity(Clamp Velocity=false, Limit Acceleration=false)
   &SubUVAnimation(
    "Sprite Renderer "=DataInterface() {propertyValues {SpriteRenderer=&self:NiagaraEmitter_0.NiagaraSpriteRendererProperties_3}}
    UseEndFrame=false
    UseStartFrame=false
   )
  ]
  rendererClasses [&NiagaraSpriteRendererProperties]
  renderers [{rendererClass=&NiagaraSpriteRendererProperties}]
 }
 Tickets {
  emitterUpdateScript [
   &EmitterState()
   &VC_FilteredSocketCount()
   &SpawnRate(SpawnRate=15, Use Spawn Probability=false)
   &SpawnRate(SpawnGroup=1, SpawnRate=17, Use Spawn Probability=false) {module=SpawnRate001}
  ]
  particleSpawnScript [
   &VC_PaintedVariantsOverride() {enabled=false}
   &InitializeParticle(
    Lifetime Max=1.5
    Lifetime Mode=::NewEnumerator1
    Position Offset=&AddVectorToPosition(
     Position=&AddVectorToPosition(Position=&AddVectorToPosition(Position=<-40 0 0>, Vector=<-30 0 0>), Vector=$User.SingleBoostPos)
     Vector=$User.DTCarOffset
    )
    Position Offset Coordinate Space=::Local
    Sprite Rotation Mode=::NewEnumerator2
    Sprite Size Max=<26 32>
    Sprite Size Min=<20.1 32>
    Sprite Size Mode=::NewEnumerator2
    UsePositionOffset=true
    Write Material Random=true
   )
   &PartitionParticles(Number Of Partitions=$Emitter.PortCount) {enabled=false}
   &SkeletalMeshLocation(
    Bone / Socket Index=$Particles.PartitionParticles.Partition
    Bone / Socket Sampling Mode=::NewEnumerator5
    Interpolate Mesh Sampling=false
    MaskSpawnGroup=false
    Mesh Sampling Type=::NewEnumerator6
    Orientation Sampling=::NewEnumerator1
    Sampled Position Offset=&AddVectorToPosition(Position=<0 0 0>, Vector=<-30 0 0>)
    Sampled Position Offset Coordinate Space:ENiagaraSkelMeshTransforms=::NewEnumerator2
    Skeletal Mesh=$System.SkeletalMesh
   ) {enabled=false}
   &AddVelocity(Velocity=&MergeHelper_VectorSwitch(
     0=&RandomRangeVector(Minimum=<-800 -128 -156>, Maximum=<-800 128 300>)
     1=&RandomRangeVector(Minimum=<-400 -200 128>, Maximum=<-400 200 256>)
     2=&RandomRangeVector(Minimum=<-400 -200 128>, Maximum=<-400 200 256>)
     3=?
     Index=$User.GameModeID
     Num=::NewEnumerator1
    ))
   &ShapeLocation(MaskSpawnGroup=false, Override Seed=false, Sphere Radius=45, UseOwnerScale=true)
   &SpriteFacingAndAlignment(Sprite Alignment=<-1 0 0>, Sprite Facing=<0 1 1>, WriteAlignment=false, WriteFacing=false) {enabled=false}
   /SetVariables(Particles.VisibilityTag=$Emitter.SpawnGroup) {isSetParameters=true}
  ]
  particleUpdateScript [
   &ParticleState()
   &ScaleSpriteSize(
    Non-Uniform Curve Sprite Scale=DataInterface() {propertyValues {
      XCurve {keys [{tangentMode=RCTM_User, time=0, value=0.5}, {tangentMode=RCTM_User, time=1, value=0.5}]}
      YCurve {keys [{tangentMode=RCTM_User, time=0, value=1}, {tangentMode=RCTM_User, time=1, value=1}]}
     }}
    Scale Sprite Size Mode=::NewEnumerator3
   )
   &AccelerationForce(Acceleration=<0 0 -325>)
   &CurlNoiseForce(
    Bias Noise=false
    Noise Frequency=25
    Noise Strength=&SelectFloatFromArray(Direct Array Index=$User.GameModeID, Float Selection Array=DataInterface() {propertyValues {FloatData [2000,3000,3000]}})
    Pan Noise=true
    Pan Noise Field=<0.83 0.5 0>
   )
   &InheritSourceMovement(
    Applied Position Scale=&VectorFromCurve(VectorCurve=DataInterface() {propertyValues {
       XCurve {keys [<at 0.3 1 RCIM_Cubic>,<at 1 0.5 RCIM_Cubic>]}
       YCurve {keys [<at 0.3 1 RCIM_Cubic>,<at 1 0.5 RCIM_Cubic>]}
       ZCurve {keys [<at 0.3 1 RCIM_Cubic>,<at 1 0.5 RCIM_Cubic>]}
      }})
    ApplyForce=false
    ApplyPosition=true
    ApplyPositionFalloff=false
   )
   &InheritVelocity(
    Inherited Velocity Amount Scale=&SelectVectorFromArray(Direct Array Index=$User.GameModeID, Vector Selection Array=DataInterface() {propertyValues {FloatData [<0 0 0>, <1 1 1>, <1 1 1>]}})
    Inherited Velocity Speed Limit=-100
    Source Speed Threshold=1000
    Speed Limit=true
    UseSpeedThreshold=false
    Velocity Source Space:ENiagaraCoordinateSpace=::Simulation
   )
   &Drag(
    Drag=&SelectFloatFromArray(Direct Array Index=$User.GameModeID, Float Selection Array=DataInterface() {propertyValues {FloatData [1,0.8,0.8]}})
    Use Linear Drag=true
    Use Rotational Drag=true
   )
   &SolveForcesAndVelocity(Clamp Velocity=false, Limit Acceleration=false)
   &JitterPosition(Jitter Delay=0.3)
   &DynamicMaterialParameters(
    Index 1 Param 1=&FloatFromCurve(FloatCurve=DataInterface() {propertyValues {Curve {keys [
         {arriveTangent=2.9016, interpMode=RCIM_Cubic, leaveTangent=2.9016, tangentMode=RCTM_User, time=0.5, value=-1}
         {arriveTangent=0.6086, interpMode=RCIM_Cubic, leaveTangent=0.6086, tangentMode=RCTM_User, time=1, value=1}
        ]}}})
    Index 1 Param 2=0.5
    Module.Index1Float0Write=true
    Module.Index1Float1Write=true
    Module.Index1Float2Write=true
    Module.Index1Float3Write=true
    Write Parameter Index 0=::NewEnumerator0
    Write Parameter Index 1=::NewEnumerator1
   )
  ]
  rendererClasses [&NiagaraSpriteRendererProperties]
  renderers [
   {rendererClass=&NiagaraSpriteRendererProperties}
   {rendererClass=&NiagaraSpriteRendererProperties, rendererIndex=1}
  ]
 }
}
summary {
 systemName=NS_Boost_AW_Optimized
 emitters [
  {emitterName=NE_Smoke001, rendererClasses [&NiagaraSpriteRendererProperties]}
  {emitterName=Cone, rendererClasses [&NiagaraSpriteRendererProperties]}
  {emitterName=Tickets, rendererClasses [&NiagaraSpriteRendererProperties]}
 ]
 userVariables [
  User.bIsFrontendPreview {type=NiagaraBool}
  User.bIsPainted {type=NiagaraBool}
  User.ColorVariantID {type=NiagaraInt32}
  User.DTCarOffset {type=Vector3f, description="", defaultValue {struct=Vector3f, value=<0 0 0>}}
  User.GameModeID {type=NiagaraInt32}
  User.PaintedColor {type=LinearColor, description="", defaultValue {struct=LinearColor, value=<rgba 1 1 1 1>}}
  User.PaintedScalar {type=NiagaraFloat, defaultValue=1, description=""}
  User.PortCount {type=NiagaraInt32, defaultValue=1}
  User.SingleBoostPos {type=Vector3f, description="", defaultValue {struct=Vector3f, value=<0 0 0>}}
 ]
}
