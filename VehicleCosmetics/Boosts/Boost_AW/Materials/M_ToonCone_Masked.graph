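# Unlit masked Niagara sprite material for the AW vehicle boost effect, drawing a toon-style flame cone tinted by particle color.
# A radial gradient, warped by panning flow noise and eroded by a panning flame mask, is multiplied into a hard edge and clipped by a rotated shape texture and particle alpha.
# A Ceil step creates two-tone emissive shading.
# Dynamic parameters expose speed, distortion and tiling through named reroutes.
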Material /VehicleCosmetics/Boosts/Boost_AW/Materials/M_ToonCone_Masked
graph {
 properties {
  bAutomaticallySetUsageInEditor=false
  bUsedWithNiagaraSprites=true
  blendMode=BLEND_Masked
  pixelDepthOffsetMode=PDOM_Legacy
  shadingModel=MSM_Unlit
  usedShadingModels=Unlit
 }
 nodes [
  n1 = NamedRerouteUsage()
  n2 = Constant() {r=0.6}
  n3 = Add(A=Constant() {r=0.4}, B=Ceil(None=Subtract(A=@n1, B=@n2)))
  n4 = ParticleColor()
  MaterialFunctionCall(Emissive=Multiply(A=Saturate(None=@n3), B=@n4.RGB)) {materialFunction=&MF_FX_EmissiveAdjustments}
  n5 = DynamicParameter() {paramNames [Speed,Distort,Tile,Param4]}
  NamedRerouteDeclaration(None=@n5.Speed) {name=Dynamic - Speed, nodeColor=<rgba 1 0 0.75 1>}
  NamedRerouteDeclaration(None=@n5.Distort) {name=Dynamic - Distort, nodeColor=<rgba 1 0 0.75 1>}
  NamedRerouteDeclaration(None=@n5.Tile) {name=Dynamic - Tile, nodeColor=<rgba 1 0 0.75 1>}
  n6 = TextureCoordinate() {uTiling=0.3, vTiling=0.6}
  n7 = NamedRerouteUsage()
  n8 = Time()
  n9 = NamedRerouteUsage()
  n10 = TextureSample(UVs=Panner(Coordinate=Multiply(A=@n6, B=@n7), Time=Multiply(A=@n8, B=@n9)) {speedX=0.5}) {texture=/DealershipSharedLibrary/FX/Textures/TexturePackage/T_VCVFX_FlameMask_RGB}
  n11 = TextureCoordinate()
  n12 = NamedRerouteDeclaration(None=Saturate(None=ComponentMask(None=@n11) {r=true})) {name=OpacityMaskUV, nodeColor=<rgba 0 1 0.7422 1>}
  n13 = Multiply(A=Subtract(A=@n12, B=Constant() {r=0.75}), B=Constant() {r=2})
  NamedRerouteDeclaration(None=Add(A=@n10.B, B=Saturate(None=@n13))) {name=OpacityMaskGradient, nodeColor=<rgba 0 1 0.7422 1>}
  NamedRerouteDeclaration(None=OneMinus(None=@n12)) {name=OpacityMaskUVInverted, nodeColor=<rgba 0 1 0.7422 1>}
  n14 = Panner(Coordinate=TextureCoordinate() {uTiling=0.5, vTiling=0.5}) {speedX=0.5}
  n15 = Multiply(
   A=ComponentMask(None=TextureSample(UVs=@n14) {texture=/DealershipSharedLibrary/FX/Textures/Noise/T_VCVFX_Flow02_N}.RGB) {g=true, r=true}
   B=NamedRerouteUsage()
  )
  n16 = Constant() {r=0.3}
  n17 = Add(A=Add(A=TextureCoordinate() {uTiling=0.8}, B=Multiply(A=@n15, B=@n16)), B=Constant2Vector() {g=-0.5, r=-0.45})
  n18 = DotProduct(A=@n17, B=@n17)
  n19 = Saturate(None=Subtract(A=OneMinus(None=@n18), B=Constant() {r=0.5}))
  n20 = LinearInterpolate(A=Multiply(A=@n10.G, B=Constant() {r=1}), B=Constant(), Alpha=NamedRerouteUsage())
  n21 = Subtract(A=Saturate(None=Subtract(A=@n19, B=@n20)), B=Constant() {r=0.333})
  n22 = Constant() {r=45}
  n23 = NamedRerouteUsage()
  n24 = Constant() {r=10}
  n25 = Multiply(A=Saturate(None=Multiply(A=@n21, B=@n22)), B=Saturate(None=Multiply(A=@n23, B=@n24)))
  n26 = TextureSample(UVs=MaterialFunctionCall(UVs=TextureCoordinate(), "Rotation Angle (0-1)"=Constant() {r=-0.25}) {materialFunction=&CustomRotator}.Rotated Values) {texture=/VehicleCosmetics/Boosts/Boost_AW/Textures/T_VCVFX_Boost_AW_Shape_M_T}
  Saturate(None=Multiply(A=@n4.A, B=Multiply(A=@n25, B=@n26.RGB)))
 ]
}
